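# Control Rig for the EssayViewMyth medium male body that adds corrective deformation and places cosmetic accessory bones.
# The shared DeformRig function runs first, then weighted parent constraints place lanyard, pendant and wristband bones.
# Calf-to-thigh rotation drives shorts bone offsets, while wrist animation attributes and distances between setup-created controls drive wristband correctives.
# Final steps shrink a spine helper, aim the pendant when IsFrontEnd is set, and offset a right-side gap bone.
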ControlRigBlueprint /BRCosmetics/Characters/Player/Male/Medium/Bodies/M_MED_EssayViewMyth/Meshes/M_MED_EssayViewMyth_CtrlRig
graphs [
 New Function_ContainedGraph {nodes [
   RigVMFunctionEntryNode()
   > &RigUnit_ParentConstraint(
    AdvancedSettings=(RotationOrderForFilter=XZY)
    Child=(Type=Bone,Name="dyn_lanyard_l")
    Parents="((Item=(Type=Bone,Name=\"deform_groin_r\"),Weight=0.0057),(Item=(Type=Bone,Name=\"thigh_twist_01_r\"),Weight=0.0106),(Item=(Type=Bone,Name=\"thigh_twist_01_l\"),Weight=0.0287),(Item=(Type=Bone,Name=\"spine_03\"),Weight=0.0297),(Item=(Type=Bone,Name=\"deform_groin_l ...(truncated)"
    Weight=1
   )
   > &RigUnit_ParentConstraint(
    AdvancedSettings=(RotationOrderForFilter=XZY)
    Child=(Type=Bone,Name="dyn_lanyard_r")
    Parents="((Item=(Type=Bone,Name=\"deform_groin_r\"),Weight=0.016),(Item=(Type=Bone,Name=\"deform_groin_l\"),Weight=0.0171),(Item=(Type=Bone,Name=\"thigh_twist_01_r\"),Weight=0.0186),(Item=(Type=Bone,Name=\"thigh_twist_01_l\"),Weight=0.0192),(Item=(Type=Bone,Name=\"spine_03\" ...(truncated)"
    Weight=1
   )
   > &RigUnit_ParentConstraint(
    AdvancedSettings=(RotationOrderForFilter=XZY)
    Child=(Type=Bone,Name="base_pendant")
    Parents=((Item=(Type=Bone,Name="neck_01"),Weight=0.0028),(Item=(Type=Bone,Name="spine_02"),Weight=0.0086),(Item=(Type=Bone,Name="spine_03"),Weight=0.0326),(Item=(Type=Bone,Name="spine_04"),Weight=0.1247),(Item=(Type=Bone,Name="spine_05"),Weight=0.8313))
    Weight=1
   )
   > &RigUnit_ParentConstraint(
    AdvancedSettings=(RotationOrderForFilter=XZY)
    Child=(Type=Bone,Name="base_wristband_r")
    Parents="((Item=(Type=Bone,Name=\"thumb_01_r\"),Weight=0.005),(Item=(Type=Bone,Name=\"ring_metacarpal_r\"),Weight=0.0143),(Item=(Type=Bone,Name=\"index_metacarpal_r\"),Weight=0.0187),(Item=(Type=Bone,Name=\"pinky_metacarpal_r\"),Weight=0.0228),(Item=(Type=Bone,Name=\"lowera ...(truncated)"
    Weight=1
   )
   > &RigUnit_ParentConstraint(
    AdvancedSettings=(RotationOrderForFilter=XZY)
    Child=(Type=Bone,Name="base_pendant")
    Parents="((Item=(Type=Bone,Name=\"deform_pec_r\"),Weight=0.0004),(Item=(Type=Bone,Name=\"clavicle_l\"),Weight=0.0005),(Item=(Type=Bone,Name=\"deform_pec_l\"),Weight=0.0087),(Item=(Type=Bone,Name=\"spine_02\"),Weight=0.0478),(Item=(Type=Bone,Name=\"spine_03\"),Weight=0.1444), ...(truncated)"
    Weight=1
   )
   > RigVMFunctionReturnNode()
  ]}
 New Function_ContainedGraph {nodes [
   RigVMFunctionEntryNode()
   > deltaRotZ = RigVMFunctionReferenceNode(Child=calf_l, Parent=thigh_l, Source Max=120, Source Min=80)
   > RigVMVariableNode(Value=@deltaRotZ.Delta) {variable=CalfRot_L}
   > deltaRotZ_1 = RigVMFunctionReferenceNode(Child=calf_r, Parent=thigh_r, Source Max=120, Source Min=80)
   > RigVMVariableNode(Value=@deltaRotZ_1.Delta) {variable=CalfRot_R}
   > RigVMFunctionReturnNode()
  ]}
 New Function_ContainedGraph {nodes [
   n1 = RigVMFunctionEntryNode()
   RigVMFunction_MathQuaternionToEuler = &RigVMFunction_MathQuaternionToEuler(
    RotationOrder=ZYX
    Value=&RigUnit_GetRelativeTransformForItem(
     Child=(Type=Bone,Name="None")
     Parent=(Type=Bone,Name="None")
     bChildInitial=False
     bParentInitial=False
     Child.Name=@n1.Child
     Parent.Name=@n1.Parent
    ).RelativeTransform.Rotation
   )
   Remap = &RigVMFunction_MathDoubleRemap(
    TargetMaximum=1
    TargetMinimum=0
    bClamp=False
    Value=@RigVMFunction_MathQuaternionToEuler.Result.Z
    SourceMinimum=@n1.Source Min
    SourceMaximum=@n1.Source Max
   )
   > RigVMFunctionReturnNode(Delta=@Remap.Result)
  ]}
 New Function_ContainedGraph {nodes [
   RigVMFunctionEntryNode()
   > &RigUnit_ModifyTransforms(
    ItemToModify=((Item=(Type=Bone,Name="base_shorts_fmd_l_02"),Transform=(Rotation=(X=-0.2164,Y=0,Z=0,W=0.9763),Translation=(X=-20,Y=0,Z=0),Scale3D=(X=0.5,Y=1,Z=1))))
    WeightMinimum=0
    Weight=RigVMVariableNode() {variable=CalfRot_L}.Value
   )
   > &RigUnit_ModifyTransforms(
    ItemToModify=((Item=(Type=Bone,Name="base_shorts_fmd_r_02"),Transform=(Rotation=(X=0.2164,Y=0,Z=0,W=0.9763),Translation=(X=-20,Y=0,Z=0),Scale3D=(X=0.5,Y=1,Z=1))))
    WeightMinimum=0
    Weight=RigVMVariableNode() {variable=CalfRot_R}.Value
   )
   > RigVMFunctionReturnNode()
  ]}
 New Function_ContainedGraph {nodes [
   RigVMFunctionEntryNode()
   > &RigUnit_ParentConstraint(
    AdvancedSettings=(InterpolationType=Average,RotationOrderForFilter=XZY)
    Child=(Type=Bone,Name="dyn_wristband_r")
    Parents=((Item=(Type=Bone,Name="lowerarm_r"),Weight=0.5),(Item=(Type=Bone,Name="hand_r"),Weight=0.5))
    Weight=1
   )
   > &RigUnit_ModifyTransforms(
    ItemToModify=((Item=(Type=Bone,Name="base_wristband_r"),Transform=(Rotation=(X=0,Y=0,Z=0,W=1),Translation=(X=0.7,Y=-0.3,Z=0),Scale3D=(X=1,Y=1,Z=1))))
    Weight=1
    WeightMinimum=0
   )
   > &RigUnit_ModifyTransforms(
    ItemToModify=((Item=(Type=Bone,Name="base_wristband_r"),Transform=(Rotation=(X=0,Y=0,Z=-0.0872,W=0.9962),Translation=(X=-1,Y=-1,Z=0),Scale3D=(X=0.9,Y=1.1,Z=1))))
    WeightMinimum=0
    Weight=RigVMDispatchNode(BoneName=root, Default=0, Name=deformRig_r_wrist_up_alpha) {template="DISPATCH_RigDispatch_GetAnimAttribute(in Name,in BoneName,in Default,out Value,out Found)"}.Value
   )
   > &RigUnit_ModifyTransforms(
    ItemToModify=((Item=(Type=Bone,Name="base_wristband_r"),Transform=(Rotation=(X=0,Y=0,Z=-0.0872,W=0.9962),Translation=(X=-1,Y=1,Z=0),Scale3D=(X=0.9,Y=1,Z=1))))
    WeightMinimum=0
    Weight=RigVMDispatchNode(BoneName=root, Default=0, Name=deformRig_r_wrist_dn_alpha) {template="DISPATCH_RigDispatch_GetAnimAttribute(in Name,in BoneName,in Default,out Value,out Found)"}.Value
   )
   n1 = &RigUnit_GetTransform(Item=(Type=Control,Name="wrist_ref_ctrl"), bInitial=False)
   RigVMFunction_MathVectorDistance = &RigVMFunction_MathVectorDistance(
    B=@n1.Transform.Translation
    A=&RigUnit_GetTransform(Item=(Type=Control,Name="wrist_down_ctrl"), bInitial=False).Transform.Translation
   )
   Remap = &RigVMFunction_MathDoubleRemap(SourceMaximum=4, SourceMinimum=15, TargetMaximum=1, TargetMinimum=0, Value=@RigVMFunction_MathVectorDistance.Result)
   > &RigUnit_ModifyTransforms(
    ItemToModify=((Item=(Type=Bone,Name="base_wristband_r"),Transform=(Rotation=(X=0,Y=0.0436,Z=0,W=0.999),Translation=(X=-0.8,Y=0,Z=-0.5),Scale3D=(X=0.9,Y=1,Z=1))))
    WeightMinimum=0
    Weight=@Remap.Result
   )
   n2 = &RigUnit_GetTransform(Item=(Type=Control,Name="wrist_up_ctrl"), bInitial=False)
   > &RigUnit_ModifyTransforms(
    ItemToModify=((Item=(Type=Bone,Name="base_wristband_r"),Transform=(Rotation=(X=0,Y=-0.0436,Z=0,W=0.999),Translation=(X=0,Y=0,Z=0.5),Scale3D=(X=1,Y=1,Z=1))))
    WeightMinimum=0
    Weight=&RigVMFunction_MathDoubleRemap(
     SourceMaximum=3
     SourceMinimum=14
     TargetMaximum=1
     TargetMinimum=0
     Value=&RigVMFunction_MathVectorDistance(A=@n1.Transform.Translation, B=@n2.Transform.Translation).Result
    ).Result
   )
   > RigVMFunctionReturnNode()
   n3 = &RigUnit_SphericalPoseReader(
    ActiveRegionScaleFactors=(PositiveWidth=0.5,NegativeWidth=1,PositiveHeight=0.5,NegativeHeight=0.5)
    ActiveRegionSize=0.1
    Debug=(bDrawDebug=true,bDraw2D=false,bDrawLocalAxes=true,DebugScale=25,DebugSegments=20,DebugThickness=0.25)
    DriverAxis=(X=1,Y=0,Z=0)
    DriverItem=(Type=Bone,Name="hand_r")
    ExecutePin=()
    FalloffRegionScaleFactors=(PositiveWidth=1,NegativeWidth=0,PositiveHeight=0.5,NegativeHeight=0.5)
    FalloffSize=0.2
    FlipHeightScaling=False
    FlipWidthScaling=true
    OptionalParentItem=(Type=Bone,Name="None")
    RotationOffset=(X=30,Y=-130,Z=0)
   )
   > n4 = &RigUnit_SphericalPoseReader(
    ActiveRegionScaleFactors=(PositiveWidth=0.5,NegativeWidth=1,PositiveHeight=0.5,NegativeHeight=0.5)
    ActiveRegionSize=0.1
    Debug=(bDrawDebug=true,bDraw2D=false,bDrawLocalAxes=true,DebugScale=25,DebugSegments=20,DebugThickness=0.25)
    DriverAxis=(X=1,Y=0,Z=0)
    DriverItem=(Type=Bone,Name="hand_r")
    FalloffRegionScaleFactors=(PositiveWidth=1,NegativeWidth=0,PositiveHeight=0.5,NegativeHeight=0.5)
    FalloffSize=0.2
    FlipHeightScaling=False
    FlipWidthScaling=false
    OptionalParentItem=(Type=Bone,Name="None")
    RotationOffset=(X=-20,Y=-30,Z=0)
   )
   > &RigUnit_ModifyTransforms(
    ItemToModify=((Item=(Type=Bone,Name="base_wristband_r"),Transform=(Rotation=(X=0,Y=0.0436,Z=0,W=0.999),Translation=(X=0.1,Y=0,Z=-0.5),Scale3D=(X=0.9,Y=1,Z=1))))
    WeightMinimum=0
    Weight=@n4.OutputParam
   )
   > &RigUnit_ModifyTransforms(
    ItemToModify=((Item=(Type=Bone,Name="base_wristband_r"),Transform=(Rotation=(X=0,Y=-0.0436,Z=0,W=0.999),Translation=(X=0,Y=0,Z=0.5),Scale3D=(X=1,Y=1,Z=1))))
    WeightMinimum=0
    Weight=@n3.OutputParam
   )
  ]}
 RigVMFunctionLibrary {nodes [
   Constraints = RigVMCollapseNode(ExecuteContext=())
   ShortsSetup = RigVMCollapseNode(ExecuteContext=())
   Variables = RigVMCollapseNode(ExecuteContext=())
   deltaRotZ = RigVMCollapseNode(ExecuteContext=())
   watchSetup = RigVMCollapseNode(ExecuteContext=())
  ]}
 RigVMModel {nodes [
   n1 = &RigUnit_PrepareForExecution()
   &RigUnit_HierarchyImportFromSkeleton(NameSpace=None, bIncludeCurves=False, bIncludeMeshSockets=False, bIncludeVirtualBones=True, ExecutePin=@n1.ExecutePin)
   > Control Stack at Item = RigVMFunctionReferenceNode(
    Add Bottom Null=False
    Add Null=true
    Add Secondary Control=False
    Control Suffix=_ctrl
    Item=(Name="hand_r",Type=Bone)
    Name=wrist_ref
    Null Suffix=_null
    Parent=(Name="hand_r",Type=Bone)
    Secondary Settings="(Shape=(bVisible=True,Name=\"Default\",Color=(R=1,G=0,B=0,A=1),Transform=(Rotation=(X=0,Y=0,Z=0,W=1),Translation=(X=0,Y=0,Z=0),Scale3D=(X=1,Y=1,Z=1))),Proxy=(bIsProxy=False,ShapeVisibility=BasedOnSelection),DisplayName=\"None\",InitialSpace=LocalSpace,bUsePref ...(truncated)"
    Settings="(Shape=(bVisible=True,Name=\"Box_Solid\",Color=(R=1,G=0,B=0,A=1),Transform=(Rotation=(X=0,Y=0,Z=0,W=1),Translation=(X=0,Y=0,Z=0),Scale3D=(X=0.25,Y=0.25,Z=0.25))),Proxy=(bIsProxy=False,ShapeVisibility=BasedOnSelection),DisplayName=\"None\",InitialSpace=LocalSpa ...(truncated)"
   )
   > SetControlOffset = /Script/ControlRig.RigUnit_SetControlOffset(
    Offset=(Rotation=(X=0,Y=0,Z=0,W=1),Translation=(X=26,Y=0,Z=0),Scale3D=(X=1,Y=1,Z=1))
    Space=LocalSpace
    Control=@"Control Stack at Item".Control.Name
   ) {is=struct RigVMUnitNode, template="SetControlOffset::Execute(in Control,io Offset,in Space)"}
   > Control Stack at Item_1 = RigVMFunctionReferenceNode(
    Add Bottom Null=False
    Add Null=true
    Add Secondary Control=False
    Control Suffix=_ctrl
    Item=(Name="hand_r",Type=Bone)
    Name=wrist_up
    Null Suffix=_null
    Parent=(Name="lowerarm_r",Type=Bone)
    Secondary Settings="(Shape=(bVisible=True,Name=\"Default\",Color=(R=1,G=0,B=0,A=1),Transform=(Rotation=(X=0,Y=0,Z=0,W=1),Translation=(X=0,Y=0,Z=0),Scale3D=(X=1,Y=1,Z=1))),Proxy=(bIsProxy=False,ShapeVisibility=BasedOnSelection),DisplayName=\"None\",InitialSpace=LocalSpace,bUsePref ...(truncated)"
    Settings="(Shape=(bVisible=True,Name=\"Pyramid_Solid\",Color=(R=0,G=0,B=1,A=1),Transform=(Rotation=(X=0,Y=0,Z=0,W=1),Translation=(X=0,Y=0,Z=0),Scale3D=(X=0.5,Y=0.5,Z=0.5))),Proxy=(bIsProxy=False,ShapeVisibility=BasedOnSelection),DisplayName=\"None\",InitialSpace=LocalSp ...(truncated)"
   )
   > &RigUnit_SetControlOffset(
    Offset=(Rotation=(X=0,Y=0,Z=0,W=1),Translation=(X=17,Y=11,Z=-14),Scale3D=(X=1,Y=1,Z=1))
    Space=LocalSpace
    Control=@"Control Stack at Item_1".Control.Name
   ) {template="SetControlOffset::Execute(in Control,io Offset,in Space)"}
   > Control Stack at Item_2 = RigVMFunctionReferenceNode(
    Add Bottom Null=False
    Add Null=true
    Add Secondary Control=False
    Control Suffix=_ctrl
    Item=(Name="hand_r",Type=Bone)
    Name=wrist_down
    Null Suffix=_null
    Parent=(Name="lowerarm_r",Type=Bone)
    Secondary Settings="(Shape=(bVisible=True,Name=\"Default\",Color=(R=1,G=0,B=0,A=1),Transform=(Rotation=(X=0,Y=0,Z=0,W=1),Translation=(X=0,Y=0,Z=0),Scale3D=(X=1,Y=1,Z=1))),Proxy=(bIsProxy=False,ShapeVisibility=BasedOnSelection),DisplayName=\"None\",InitialSpace=LocalSpace,bUsePref ...(truncated)"
    Settings="(Shape=(bVisible=True,Name=\"Pyramid_Solid\",Color=(R=0,G=1,B=0,A=1),Transform=(Rotation=(X=0,Y=0,Z=0,W=1),Translation=(X=0,Y=0,Z=0),Scale3D=(X=0.5,Y=0.5,Z=0.5))),Proxy=(bIsProxy=False,ShapeVisibility=BasedOnSelection),DisplayName=\"None\",InitialSpace=LocalSp ...(truncated)"
   )
   > &RigUnit_SetControlOffset(
    Offset=(Rotation=(X=0,Y=0,Z=0,W=1),Translation=(X=10,Y=-8,Z=20),Scale3D=(X=1,Y=1,Z=1))
    Space=LocalSpace
    Control=@"Control Stack at Item_2".Control.Name
   ) {template="SetControlOffset::Execute(in Control,io Offset,in Space)"}
   RigUnit_BeginExecution = &RigUnit_BeginExecution()
   FN_DeformRig_FNC = RigVMFunctionReferenceNode(
    DeformRig_ Curve="(deformRig_l_upperarm_up_alpha_72_#1=true,deformRig_r_upperarm_up_alpha_73_#2=true,deformRig_l_upperarm_fr_alpha_74_#3=true,deformRig_r_upperarm_fr_alpha_75_CB51B80C4 ...(truncated)"
    DeformRig_Bool="(upperarm_1_#4=True,elbow_3_#5=True,wrist_7_#6=True,groin_8_#7=True,glute_17_#8=True,knee_9_40C408324CA261 ...(truncated)"
    DeformRig_Debug="(upperarm_debug=False,wrist_debug=False,lat_debug=False,groinGlute_debug=False,upperarm_debug_18_#4=False,wrist_debug_20_#6=False,groinGlute_debug_27_#7=False,lat_deb ...(truncated)"
    DeformRig_Value="(upperarm_up_18_#9=(Translation=(X=0,Y=0,Z=8),Scale3D=(X=0.7,Y=1,Z=1)),upperarm_fwd_fr_19_#10=(Translation=(X=0,Y=10,Z=0),Scale3D=(X=0.5,Y=1,Z=1)),upperarm_fwd_bk_20_9973AAD84EE0A1E349BC9E85067D4A3 ...(truncated)"
    ExecuteContext=@RigUnit_BeginExecution.ExecutePin
   )
   > Constraints = RigVMFunctionReferenceNode()
   > Variables = RigVMFunctionReferenceNode()
   > ShortsSetup = RigVMFunctionReferenceNode()
   > watchSetup = RigVMFunctionReferenceNode()
   > &RigUnit_ModifyTransforms(
    ItemToModify=((Item=(Type=Bone,Name="spine_04_jnt_helper"),Transform=(Rotation=(X=0,Y=0,Z=0,W=1),Translation=(X=0,Y=0,Z=0),Scale3D=(X=0.7,Y=0.7,Z=0.7))))
    Weight=1
    WeightMinimum=0
   )
   RigVMFunction_MathBoolToFloat = &RigVMFunction_MathBoolToFloat(Value=RigVMVariableNode() {variable=IsFrontEnd}.Value)
   > &RigUnit_ParentConstraint(
    AdvancedSettings=(RotationOrderForFilter=XZY)
    Child=(Type=Bone,Name="base_pendant_for_aim")
    Parents="((Item=(Type=Bone,Name=\"deform_pec_r\"),Weight=0.0013),(Item=(Type=Bone,Name=\"deform_pec_l\"),Weight=0.0013),(Item=(Type=Bone,Name=\"pelvis\"),Weight=0.0052),(Item=(Type=Bone,Name=\"spine_01\"),Weight=0.0366),(Item=(Type=Bone,Name=\"spine_05\"),Weight=0.0742),(Ite ...(truncated)"
    Weight=@RigVMFunction_MathBoolToFloat.Result
   )
   > &RigUnit_AimConstraintLocalSpaceOffset(
    Child=(Type=Bone,Name="base_pendant")
    Filter=(bX=True,bY=True,bZ=True)
    Parents=((Item=(Type=Bone,Name="base_pendant_for_aim"),Weight=1))
    WorldUp=(Target=(X=0,Y=0,Z=1),Kind=Direction,Space=(Type=Bone,Name="base_pendant_for_aim"))
    Weight=@RigVMFunction_MathBoolToFloat.Result
   )
   > &RigUnit_ParentConstraint(
    AdvancedSettings=(RotationOrderForFilter=XZY)
    Child=(Type=Bone,Name="dyn_offset_gap_r")
    Parents="((Item=(Type=Bone,Name=\"spine_04\"),Weight=0.0003),(Item=(Type=Bone,Name=\"upperarm_twist_01_r\"),Weight=0.0009),(Item=(Type=Bone,Name=\"spine_03\"),Weight=0.001),(Item=(Type=Bone,Name=\"deform_upperarm_up_r\"),Weight=0.0013),(Item=(Type=Bone,Name=\"neck_01\"),Weig ...(truncated)"
    Weight=1
   )
   > &RigUnit_ModifyTransforms(
    ItemToModify=((Item=(Type=Bone,Name="dyn_offset_gap_r"),Transform=(Rotation=(X=0,Y=0,Z=0,W=1),Translation=(X=1.3,Y=0,Z=0),Scale3D=(X=1,Y=1,Z=1))))
    Weight=1
    WeightMinimum=0
   )
  ]}
]
hierarchy [
 root {loc=<0 0 0>, rot=<0 0 0 1>}
 pelvis {parent=root, loc=<0 -0.7071 93.1191>, rot=<0.0127 0.707 -0.0127 -0.707>}
 spine_01 {parent=pelvis, loc=<6.7494 -0.6388 0>, rot=<0 0 0.034 -0.9994>}
 spine_02 {parent=spine_01, loc=<7.2518 -0.035 0>, rot=<0 0 0.0238 -0.9997>}
 spine_03 {parent=spine_02, loc=<6.0104 0 0>, rot=<0 0 0.0149 -0.9999>}
 spine_04 {parent=spine_03, loc=<8.1024 0 0>, rot=<0 0 -0.0132 -0.9999>}
 spine_05 {parent=spine_04, loc=<10.0146 0 0>, rot=<0 0 -0.0015 -1>}
 neck_01 {parent=spine_05, loc=<11.3138 0.1529 0>, rot=<0 0 -0.2003 -0.9797>}
 neck_02 {parent=neck_01, loc=<4.5627 0 0>, rot=<0 0 0.0504 -0.9987>}
 head {parent=neck_02, loc=<5.3914 0 0>, rot=<0 0 0.115 -0.9934>}
 attach_cape {parent=spine_05, loc=<9.017 -11.714 0>, rot=<0.0616 -0.7044 -0.0616 -0.7044>}
 cape {parent=attach_cape, type=Null, loc=<0 0 0>, rot=<0 0 0 1>}
 attach_backpack {parent=spine_05, loc=<-0.123 -13.473 0>, rot=<-0.0093 -0.707 0.0093 -0.707>}
 Backpack_BR {parent=attach_backpack, type=Null, loc=<0 1.23 0>, rot=<0 0 0 1>}
 clavicle_l {parent=spine_05, loc=<4.3575 -0.1466 -2.3223>, rot=<0.0465 -0.7289 -0.009 -0.6829>}
 upperarm_l {parent=clavicle_l, loc=<15.5111 0 0>, rot=<0.0226 -0.3234 -0.0088 -0.9459>}
 lowerarm_l {parent=upperarm_l, loc=<26.6514 0 0>, rot=<0.0391 -0.0099 -0.2453 -0.9686>}
 hand_l {parent=lowerarm_l, loc=<22.3491 0 0>, rot=<-0.7081 -0.0507 0.0975 -0.6975>}
 thumb_01_l {parent=hand_l, loc=<1.3118 -0.3191 -2.459>, rot=<0.6546 -0.4042 0.0664 -0.6354>}
 thumb_02_l {parent=thumb_01_l, loc=<5.9336 -0.0448 0.0401>, rot=<-0.0037 0.0213 0.298 -0.9543>}
 thumb_03_l {parent=thumb_02_l, loc=<2.2223 0.3682 -0.0907>, rot=<0.0118 -0.013 0.0066 -0.9998>}
 index_metacarpal_l {parent=hand_l, loc=<1.6535 0.0183 -2.96>, rot=<0.0158 -0.0124 -0.0173 -0.9996>}
 index_01_l {parent=index_metacarpal_l, loc=<8.3008 0.0357 -0.0524>, rot=<-0.0044 0.0485 0.0924 -0.9945>}
 index_02_l {parent=index_01_l, loc=<3.3587 -0.3004 0.0452>, rot=<-0.002 0.0027 0.3225 -0.9466>}
 index_03_l {parent=index_02_l, loc=<2.5578 -0.123 -0.0252>, rot=<-0.002 0.0067 0.2558 -0.9667>}
 middle_metacarpal_l {parent=hand_l, loc=<2.1856 -0.1725 -0.5984>, rot=<0.006 0.0168 -0.0329 -0.9993>}
 middle_01_l {parent=middle_metacarpal_l, loc=<7.3275 -0.2447 0.0129>, rot=<-0.0208 0.0042 0.0858 -0.9961>}
 middle_02_l {parent=middle_01_l, loc=<4.0123 0.1401 0.1175>, rot=<-0.0218 0.0258 0.3512 -0.9357>}
 middle_03_l {parent=middle_02_l, loc=<2.6077 0.0534 0.1373>, rot=<0.0082 0.0188 0.1708 -0.9851>}
 ring_metacarpal_l {parent=hand_l, loc=<2.2074 -0.3625 1.3638>, rot=<-0.0284 0.0722 -0.0458 -0.9959>}
 ring_01_l {parent=ring_metacarpal_l, loc=<7.0295 -0.6197 -0.1446>, rot=<-0.0489 -0.0252 0.1892 -0.9804>}
 ring_02_l {parent=ring_01_l, loc=<3.6174 0.4647 -0.1027>, rot=<0 0 0.2806 -0.9598>}
 ring_03_l {parent=ring_02_l, loc=<2.4152 0.114 0.0199>, rot=<0 0 0.1189 -0.9929>}
 pinky_metacarpal_l {parent=hand_l, loc=<1.7616 -0.9753 3.3814>, rot=<-0.0397 0.0814 -0.069 -0.9935>}
 pinky_01_l {parent=pinky_metacarpal_l, loc=<7.1709 -0.4831 -0.1094>, rot=<-0.0654 -0.0035 0.2242 -0.9723>}
 pinky_02_l {parent=pinky_01_l, loc=<2.8274 0.0893 -0.1043>, rot=<0.0037 0.0011 0.2854 -0.9584>}
 pinky_03_l {parent=pinky_02_l, loc=<2.2046 0.0307 -0.0078>, rot=<-0.0002 0.0098 0.1445 -0.9895>}
 deform_wrist_dn_base_l {parent=hand_l, loc=<0 -2.2342 0>, rot=<0 0 0 -1>}
 deform_wrist_dn_l {parent=deform_wrist_dn_base_l, loc=<0 0 0>, rot=<0 0 0 1>}
 deform_wrist_up_base_l {parent=hand_l, loc=<0 2.2682 0>, rot=<0 0 0 -1>}
 deform_wrist_up_l {parent=deform_wrist_up_base_l, loc=<0 0 0>, rot=<0 0 0 1>}
 weapon_l {parent=hand_l, loc=<-0.1974 -3.4111 0.3811>, rot=<0 0 0 1>}
 hand_attach_l {parent=hand_l, loc=<8.5 -3 0>, rot=<0.7071 0.7071 0 0>}
 lowerarm_twist_02_l {parent=lowerarm_l, loc=<9.6199 0 0>, rot=<0 0 0 -1>}
 lowerarm_twist_01_l {parent=lowerarm_l, loc=<17.8549 0 0>, rot=<0 0 0 -1>}
 deform_elbow_in_base_l {parent=lowerarm_l, loc=<0.2112 5.259 0.8274>, rot=<0 0 0 -1>}
 deform_elbow_in_l {parent=deform_elbow_in_base_l, loc=<0 0 0>, rot=<0 0 0 1>}
 deform_elbow_out_base_l {parent=lowerarm_l, loc=<0 -4 0>, rot=<0 0 0 -1>}
 deform_elbow_out_l {parent=deform_elbow_out_base_l, loc=<0 0 0>, rot=<0 0 0 1>}
 upperarm_twist_01_l {parent=upperarm_l, loc=<0.0041 0 0>, rot=<0 0 0 -1>}
 deform_upperarm_bk_base_l {parent=upperarm_twist_01_l, loc=<1.0813 -7.2193 -1.9453>, rot=<0 0 0 1>}
 deform_upperarm_bk_l {parent=deform_upperarm_bk_base_l, loc=<0 0 0>, rot=<0 0 0 1>}
 deform_upperarm_fr_base_l {parent=upperarm_twist_01_l, loc=<-0.3691 7.0506 -2.765>, rot=<0 0 0 1>}
 deform_upperarm_fr_l {parent=deform_upperarm_fr_base_l, loc=<0 0 0>, rot=<0 0 0 1>}
 deform_upperarm_up_base_l {parent=upperarm_twist_01_l, loc=<-3.2216 -1.1391 3.3311>, rot=<0 0 0 1>}
 deform_upperarm_up_l {parent=deform_upperarm_up_base_l, loc=<0 0 0>, rot=<0 0 0 1>}
 upperarm_twist_02_l {parent=upperarm_l, loc=<13.7941 0 0>, rot=<0 0 0 -1>}
 clavicle_r {parent=spine_05, loc=<4.3575 -0.1466 2.3223>, rot=<0.7289 0.0465 -0.6829 0.009>}
 upperarm_r {parent=clavicle_r, loc=<-15.5111 0 0>, rot=<0.0226 -0.3234 -0.0088 -0.9459>}
 lowerarm_r {parent=upperarm_r, loc=<-26.6514 0 0>, rot=<0.0391 -0.0099 -0.2453 -0.9686>}
 hand_r {parent=lowerarm_r, loc=<-22.3491 0 0>, rot=<-0.7081 -0.0507 0.0975 -0.6975>}
 thumb_01_r {parent=hand_r, loc=<-1.3118 0.3191 2.459>, rot=<0.6546 -0.4042 0.0664 -0.6354>}
 thumb_02_r {parent=thumb_01_r, loc=<-5.9336 0.0448 -0.0401>, rot=<-0.0037 0.0213 0.298 -0.9543>}
 thumb_03_r {parent=thumb_02_r, loc=<-2.2223 -0.3682 0.0907>, rot=<0.0118 -0.013 0.0066 -0.9998>}
 index_metacarpal_r {parent=hand_r, loc=<-1.6535 -0.0183 2.96>, rot=<0.0158 -0.0124 -0.0173 -0.9996>}
 index_01_r {parent=index_metacarpal_r, loc=<-8.3008 -0.0357 0.0524>, rot=<-0.0044 0.0485 0.0924 -0.9945>}
 index_02_r {parent=index_01_r, loc=<-3.3587 0.3004 -0.0452>, rot=<-0.002 0.0027 0.3225 -0.9466>}
 index_03_r {parent=index_02_r, loc=<-2.5578 0.123 0.0252>, rot=<-0.002 0.0067 0.2558 -0.9667>}
 middle_metacarpal_r {parent=hand_r, loc=<-2.1856 0.1725 0.5984>, rot=<0.006 0.0168 -0.0329 -0.9993>}
 middle_01_r {parent=middle_metacarpal_r, loc=<-7.3275 0.2447 -0.0129>, rot=<-0.0208 0.0042 0.0858 -0.9961>}
 middle_02_r {parent=middle_01_r, loc=<-4.0123 -0.1401 -0.1175>, rot=<-0.0218 0.0258 0.3512 -0.9357>}
 middle_03_r {parent=middle_02_r, loc=<-2.6077 -0.0534 -0.1373>, rot=<0.0082 0.0188 0.1708 -0.9851>}
 ring_metacarpal_r {parent=hand_r, loc=<-2.2074 0.3625 -1.3638>, rot=<-0.0284 0.0722 -0.0458 -0.9959>}
 ring_01_r {parent=ring_metacarpal_r, loc=<-7.0295 0.6197 0.1446>, rot=<-0.0489 -0.0252 0.1892 -0.9804>}
 ring_02_r {parent=ring_01_r, loc=<-3.6174 -0.4647 0.1027>, rot=<0 0 0.2806 -0.9598>}
 ring_03_r {parent=ring_02_r, loc=<-2.4152 -0.114 -0.0199>, rot=<0 0 0.1189 -0.9929>}
 pinky_metacarpal_r {parent=hand_r, loc=<-1.7616 0.9753 -3.3814>, rot=<-0.0397 0.0814 -0.069 -0.9935>}
 pinky_01_r {parent=pinky_metacarpal_r, loc=<-7.1709 0.4831 0.1094>, rot=<-0.0654 -0.0035 0.2242 -0.9723>}
 pinky_02_r {parent=pinky_01_r, loc=<-2.8274 -0.0893 0.1043>, rot=<0.0037 0.0011 0.2854 -0.9584>}
 pinky_03_r {parent=pinky_02_r, loc=<-2.2046 -0.0307 0.0078>, rot=<-0.0002 0.0098 0.1445 -0.9895>}
 deform_wrist_dn_base_r {parent=hand_r, loc=<-0.0004 2.2345 -0.0001>, rot=<0 0 0 -1>}
 deform_wrist_dn_r {parent=deform_wrist_dn_base_r, loc=<0 0 0>, rot=<0 0 0 1>}
 deform_wrist_up_base_r {parent=hand_r, loc=<0.0003 -2.2684 0.0001>, rot=<0 0 0 -1>}
 deform_wrist_up_r {parent=deform_wrist_up_base_r, loc=<0 0 0>, rot=<0 0 0 1>}
 base_wristband_r {parent=hand_r, loc=<1.007 0 0>, rot=<0 0 0 -1>}
 dyn_wristband_r {parent=base_wristband_r, loc=<0 0 0>, rot=<0 0 0 1>}
 weapon_r {parent=hand_r, loc=<0.1974 3.4111 -0.3811>, rot=<0 0 0 1>}
 hand_attach_r {parent=hand_r, loc=<-8.5001 2.9999 0>, rot=<0 0 -0.7071 -0.7071>}
 lowerarm_twist_02_r {parent=lowerarm_r, loc=<-9.6199 0 0>, rot=<0 0 0 -1>}
 lowerarm_twist_01_r {parent=lowerarm_r, loc=<-17.8549 0 0>, rot=<0 0 0 -1>}
 deform_elbow_in_base_r {parent=lowerarm_r, loc=<-0.2111 -5.2591 -0.8275>, rot=<0 0 0 -1>}
 deform_elbow_in_r {parent=deform_elbow_in_base_r, loc=<-0.0001 0 0.0002>, rot=<0 0 0 1>}
 deform_elbow_out_base_r {parent=lowerarm_r, loc=<0 4 -0.0001>, rot=<0 0 0 -1>}
 deform_elbow_out_r {parent=deform_elbow_out_base_r, loc=<0 0 0>, rot=<0 0 0 1>}
 upperarm_twist_01_r {parent=upperarm_r, loc=<-0.0041 0 0>, rot=<0 0 0 -1>}
 deform_upperarm_bk_base_r {parent=upperarm_twist_01_r, loc=<-1.0814 7.2192 1.9455>, rot=<0 0 0 1>}
 deform_upperarm_bk_r {parent=deform_upperarm_bk_base_r, loc=<0 0 0>, rot=<0 0 0 1>}
 deform_upperarm_fr_base_r {parent=upperarm_twist_01_r, loc=<0.3689 -7.0506 2.7652>, rot=<0 0 0 1>}
 deform_upperarm_fr_r {parent=deform_upperarm_fr_base_r, loc=<0 0 0>, rot=<0 0 0 1>}
 deform_upperarm_up_base_r {parent=upperarm_twist_01_r, loc=<3.2214 1.1391 -3.3308>, rot=<0 0 0 1>}
 deform_upperarm_up_r {parent=deform_upperarm_up_base_r, loc=<0 0 0>, rot=<0 0 0 -1>}
 upperarm_twist_02_r {parent=upperarm_r, loc=<-13.7941 0 0>, rot=<0 0 0 -1>}
 base_pendant {parent=spine_05, loc=<5.1833 8.6872 0>, rot=<0.1662 0.6873 0.6873 0.1662>}
 dyn_pendant_1 {parent=base_pendant, loc=<0 0 0>, rot=<0 0 0 1>}
 dyn_pendant_2 {parent=dyn_pendant_1, loc=<8.5892 0 0.3844>, rot=<0 0 0 1>}
 dyn_pendant_3 {parent=dyn_pendant_2, loc=<2.2385 0 -0.3314>, rot=<0 -0.1749 0 -0.9846>}
 base_pendant_for_aim {parent=spine_05, loc=<-7.2488 15.074 0>, rot=<0.1662 0.6873 0.6873 0.1662>}
 dyn_offset_gap_r {parent=spine_05, loc=<13.2012 -1.5191 7.835>, rot=<0.0601 -0.9691 -0.2239 -0.0842>}
 deform_pec_aim_base_l {parent=spine_04, loc=<7.8732 12.6388 -2.4123>, rot=<0.0102 -0.4602 0.3979 -0.7936>}
 deform_pec_aim_l {parent=deform_pec_aim_base_l, loc=<0 0 0>, rot=<0 0 0 1>}
 deform_pec_base_l {parent=deform_pec_aim_l, loc=<10.766 0 0>, rot=<0 0 0 1>}
 deform_pec_l {parent=deform_pec_base_l, loc=<0.055 0 0>, rot=<0 0 0 1>}
 deform_pec_aim_base_r {parent=spine_04, loc=<7.8727 12.6388 2.4123>, rot=<-0.4602 -0.0102 0.7936 0.3979>}
 deform_pec_aim_r {parent=deform_pec_aim_base_r, loc=<0 0 0>, rot=<0 0 0 1>}
 deform_pec_base_r {parent=deform_pec_aim_r, loc=<-10.7662 -0.0006 -0.0007>, rot=<0 0 0 1>}
 deform_pec_r {parent=deform_pec_base_r, loc=<0 0 0>, rot=<0 0 0 1>}
 spine_04_jnt_helper {parent=spine_04, loc=<0 0 0>, rot=<0 0 0 1>}
 thigh_l {parent=pelvis, loc=<-2.5693 -0.0109 -7.7501>, rot=<0.0012 0.0572 0.0155 -0.9982>}
 calf_l {parent=thigh_l, loc=<-39.0484 0 0>, rot=<0.0006 -0.0161 -0.0391 -0.9991>}
 foot_l {parent=calf_l, loc=<-42.561 0.1077 0.4292>, rot=<0.0418 -0.0455 0.0343 -0.9975>}
 ball_l {parent=foot_l, loc=<-6.1599 13.8264 0.6806>, rot=<0.0001 -0.0001 -0.7063 -0.7079>}
 calf_twist_02_l {parent=calf_l, loc=<-18.8064 -0.0967 0.0397>, rot=<0.0237 0.0095 -0.0035 -0.9997>}
 calf_twist_01_l {parent=calf_l, loc=<-35.813 -0.0967 0.0397>, rot=<0.0237 0.0095 -0.0035 -0.9997>}
 deform_knee_in_base_l {parent=calf_l, loc=<1.298 -5.2957 -0.0496>, rot=<0 0 0 -1>}
 deform_knee_in_l {parent=deform_knee_in_base_l, loc=<0 0 0>, rot=<0 0 0 1>}
 deform_knee_out_base_l {parent=calf_l, loc=<0 4 0>, rot=<0 0 0 -1>}
 deform_knee_out_l {parent=deform_knee_out_base_l, loc=<0 0 0>, rot=<0 0 0 1>}
 thigh_twist_01_l {parent=thigh_l, loc=<-21.6303 0 0>, rot=<0.0013 0 0 -1>}
 base_shorts_fmd_l {parent=thigh_twist_01_l, loc=<3.6474 -0.0229 -0.0434>, rot=<0 0 1 0>}
 base_shorts_fmd_l_01 {parent=base_shorts_fmd_l, loc=<1.116 -10.2219 -1.1131>, rot=<-0.7206 0.0074 -0.0042 -0.6933>}
 dyn_shorts_fmd_l_01 {parent=base_shorts_fmd_l_01, loc=<0 0 0>, rot=<0 0 0 1>}
 base_shorts_fmd_l_04 {parent=base_shorts_fmd_l, loc=<0.6419 -1.3875 8.7348>, rot=<-0.1202 -0.0187 0.0007 -0.9926>}
 dyn_shorts_fmd_l_04 {parent=base_shorts_fmd_l_04, loc=<0 0 0>, rot=<0 0 0 1>}
 base_shorts_fmd_l_03 {parent=base_shorts_fmd_l, loc=<1.1918 0.9247 -9.467>, rot=<0.9973 -0.0018 0.0083 -0.073>}
 dyn_shorts_fmd_l_03 {parent=base_shorts_fmd_l_03, loc=<0 0 0>, rot=<0 0 0 1>}
 base_shorts_fmd_l_02 {parent=base_shorts_fmd_l, loc=<1.2019 9.9064 -0.4382>, rot=<0.7096 -0.0203 0.0199 -0.704>}
 dyn_shorts_fmd_l_02 {parent=base_shorts_fmd_l_02, loc=<0 0 0>, rot=<0 0 0 1>}
 deform_groin_base_l {parent=thigh_l, loc=<-2.1908 8.1142 -1.7442>, rot=<0 0 0 -1>}
 deform_groin_l {parent=deform_groin_base_l, loc=<0 0 0>, rot=<0 0 0 1>}
 thigh_r {parent=pelvis, loc=<-2.5693 -0.0109 7.7501>, rot=<0.0572 -0.0012 0.9982 0.0155>}
 calf_r {parent=thigh_r, loc=<39.0484 0 0>, rot=<0.0006 -0.0161 -0.0391 -0.9991>}
 foot_r {parent=calf_r, loc=<42.561 -0.1077 -0.4292>, rot=<0.0418 -0.0455 0.0343 -0.9975>}
 ball_r {parent=foot_r, loc=<6.1599 -13.8264 -0.6806>, rot=<0.0001 -0.0001 -0.7063 -0.7079>}
 calf_twist_02_r {parent=calf_r, loc=<18.8064 0.0967 -0.0397>, rot=<0.0237 0.0095 -0.0035 -0.9997>}
 calf_twist_01_r {parent=calf_r, loc=<35.813 0.0967 -0.0397>, rot=<0.0237 0.0095 -0.0035 -0.9997>}
 deform_knee_in_base_r {parent=calf_r, loc=<-1.298 5.2957 0.0496>, rot=<0 0 0 -1>}
 deform_knee_in_r {parent=deform_knee_in_base_r, loc=<0 0 0>, rot=<0 0 0 1>}
 deform_knee_out_base_r {parent=calf_r, loc=<0 -4 0>, rot=<0 0 0 -1>}
 deform_knee_out_r {parent=deform_knee_out_base_r, loc=<0 0 0>, rot=<0 0 0 1>}
 thigh_twist_01_r {parent=thigh_r, loc=<21.6303 0 0>, rot=<0.0013 0 0 -1>}
 base_shorts_fmd_r {parent=thigh_twist_01_r, loc=<-3.6654 0.023 0.0457>, rot=<0 0 0 -1>}
 base_shorts_fmd_r_04 {parent=base_shorts_fmd_r, loc=<0.633 -1.3873 -8.8093>, rot=<0.9921 -0.0331 0.0217 -0.1189>}
 dyn_shorts_fmd_r_04 {parent=base_shorts_fmd_r_04, loc=<0 0 0>, rot=<0 0 0 1>}
 base_shorts_fmd_r_03 {parent=base_shorts_fmd_r, loc=<1.2022 1.1458 9.2372>, rot=<-0.0739 0.006 0.0185 -0.9971>}
 dyn_shorts_fmd_r_03 {parent=base_shorts_fmd_r_03, loc=<0 0 0>, rot=<0 0 0 1>}
 base_shorts_fmd_r_02 {parent=base_shorts_fmd_r, loc=<1.2016 9.9064 0.4382>, rot=<0.7119 -0.0161 0.0222 -0.7017>}
 dyn_shorts_fmd_r_02 {parent=base_shorts_fmd_r_02, loc=<0 0 0>, rot=<0 0 0 1>}
 base_shorts_fmd_r_01 {parent=base_shorts_fmd_r, loc=<1.1039 -10.2215 1.0109>, rot=<-0.7221 -0.0039 0.0131 -0.6916>}
 dyn_shorts_fmd_r_01 {parent=base_shorts_fmd_r_01, loc=<0 0 0>, rot=<0 0 0 1>}
 deform_groin_base_r {parent=thigh_r, loc=<2.1908 -8.1142 1.7442>, rot=<0 0 0 -1>}
 deform_groin_r {parent=deform_groin_base_r, loc=<0 0 0>, rot=<0 0 0 1>}
 deform_glute_base_l {parent=pelvis, loc=<-8.8064 -8.8222 -8.4287>, rot=<0.0275 0.0501 -0.4472 -0.8926>}
 deform_glute_l {parent=deform_glute_base_l, loc=<0 0 0>, rot=<0 0 0 1>}
 deform_glute_base_r {parent=pelvis, loc=<-8.8062 -8.8222 8.4287>, rot=<0.0501 -0.0275 0.8926 -0.4472>}
 deform_glute_r {parent=deform_glute_base_r, loc=<0 0 0>, rot=<0 0 0 1>}
 base_lanyard {parent=pelvis, loc=<4.3907 10.7437 -0.6136>, rot=<0.0299 0.7065 0.7065 0.0299>}
 dyn_lanyard_main {parent=base_lanyard, loc=<0 0 0>, rot=<0 0 0 1>}
 dyn_lanyard_l {parent=dyn_lanyard_main, loc=<0.2087 -2.1019 0.0858>, rot=<0 0 0 1>}
 dyn_lanyard_r {parent=dyn_lanyard_main, loc=<0.2424 2.1577 0.0842>, rot=<0 0 0 1>}
 attach {parent=root, loc=<0 0 0>, rot=<0 0 0 1>}
 ik_hand_root {parent=root, loc=<0 0 0>, rot=<0 0 0 1>}
 ik_hand_gun {parent=ik_hand_root, loc=<-51.9727 4.0883 102.6935>, rot=<0.6084 0.1793 0.4613 -0.6205>}
 ik_hand_r {parent=ik_hand_gun, loc=<0 0 0>, rot=<0 0 0 1>}
 ik_hand_l {parent=ik_hand_gun, loc=<53.0285 82.1781 35.2051>, rot=<-0.5102 -0.7906 -0.3387 0>}
 ik_foot_root {parent=root, loc=<0 0 0>, rot=<0 0 0 1>}
 ik_foot_l {parent=ik_foot_root, loc=<15.2608 -4.4176 9.4432>, rot=<0.0357 0.7045 0.0284 -0.7082>}
 ik_foot_r {parent=ik_foot_root, loc=<-15.2608 -4.4176 9.4432>, rot=<0.7082 0.0284 -0.7045 0.0357>}
 root_fp {parent=root, loc=<0 0 0>, rot=<0 0 0 1>}
 attach_fp {parent=root_fp, loc=<0 0 0>, rot=<0 0 0 1>}
 camera_root {parent=root_fp, loc=<0 0 0>, rot=<0 0 0 1>}
 camera {parent=camera_root, loc=<0 0 0>, rot=<0 0 0 1>}
 attach_emote_01 {parent=root, loc=<0 0 0>, rot=<0 0 0 1>}
 attach_emote_02 {parent=root, loc=<0 0 0>, rot=<0 0 0 1>}
 attach_emote_03 {parent=root, loc=<0 0 0>, rot=<0 0 0 1>}
 warp_point_01 {parent=root, loc=<0 0 0>, rot=<0 0 0 1>}
 warp_point_02 {parent=root, loc=<0 0 0>, rot=<0 0 0 1>}
 warp_point_03 {parent=root, loc=<0 0 0>, rot=<0 0 0 1>}
]
variables [
 CalfRot_L {type=double, default=0}
 CalfRot_R {type=double, default=0}
 IsFrontEnd {type=bool, default=False}
]
