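# Post-body AnimBP for the EssayViewMyth face-accessory mesh, simulating its hair locks on the player.
# Copied parent and body-mesh poses feed a crouch-aware Control Rig, cached for three spine_05-space RigidBody branches.
# Backbling, capes or front-end skydiving select a constraint-pinned back-lock setup; otherwise the Pier or default physics asset runs with bone offsets.
# The EventGraph derives these flags and overrides gravity for skydiving, parachute and backbling states.
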AnimBlueprint /BRCosmetics/Characters/Player/Male/Medium/Bodies/M_MED_EssayViewMyth/Meshes/Parts/M_MED_EssayViewMyth_FaceAcc_AnimBP
target_skeleton=&/M_MED_EssayViewMyth_FaceAcc_Skeleton
graphs [
 AnimGraph {nodes [
   "Constraint - base_M_backLock_1 (PPPP)"!(
    ComponentPose
    Alpha=1
    ConstraintWeights_0=0.8674
    ConstraintWeights_1:float=0.0689
    ConstraintWeights_2:float=0.047
    ConstraintWeights_3:float=0.0167
   ) -> (Pose) {
    alphaBoolBlend {blendOption=Linear}
    boneToModify {boneName=base_M_backLock_1}
    constraintSetup [{targetBone {boneName=spine_05}}, {targetBone {boneName=clavicle_r}}, {targetBone {boneName=neck_01}}, {targetBone {boneName=clavicle_l}}]
    constraintWeights [1,1,1,1]
   }
   "Constraint - base_M_backLock_2 (PPP)"!(ComponentPose, Alpha=1, ConstraintWeights_0=0.85, ConstraintWeights_1:float=0.1169, ConstraintWeights_2:float=0.0417) -> (Pose) {
    alphaBoolBlend {blendOption=Linear}
    boneToModify {boneName=base_M_backLock_2}
    constraintSetup [{targetBone {boneName=spine_05}}, {targetBone {boneName=clavicle_r}}, {targetBone {boneName=neck_01}}]
    constraintWeights [1,1,1]
   }
   "Constraint - base_M_backLock_3 (PPPPP)"!(
    ComponentPose
    Alpha=1
    ConstraintWeights_0=0.9376
    ConstraintWeights_1:float=0.0516
    ConstraintWeights_2:float=0.0031
    ConstraintWeights_3:float=0.0029
    ConstraintWeights_4:float=0.0022
   ) -> (Pose) {
    alphaBoolBlend {blendOption=Linear}
    boneToModify {boneName=base_M_backLock_3}
    constraintSetup [
     {targetBone {boneName=spine_05}}
     {targetBone {boneName=clavicle_l}}
     {targetBone {boneName=neck_01}}
     {targetBone {boneName=spine_03}}
     {targetBone {boneName=spine_04}}
    ]
    constraintWeights [1,1,1,1,1]
   }
   "Constraint - base_L_backLock_2 (PPPPP)"!(
    ComponentPose
    Alpha=1
    ConstraintWeights_0=0.8745
    ConstraintWeights_1:float=0.1094
    ConstraintWeights_2:float=0.0075
    ConstraintWeights_3:float=0.003
    ConstraintWeights_4:float=0.0026
   ) -> (Pose) {
    alphaBoolBlend {blendOption=Linear}
    boneToModify {boneName=base_L_backLock_2}
    constraintSetup [
     {targetBone {boneName=spine_05}}
     {targetBone {boneName=clavicle_l}}
     {targetBone {boneName=neck_01}}
     {targetBone {boneName=spine_03}}
     {targetBone {boneName=spine_04}}
    ]
    constraintWeights [1,1,1,1,1]
   }
   "Constraint - dyn_L_backLock_1_04 (R)"(Alpha=0.5, ConstraintWeights_0=1) {
    alphaBoolBlend {blendOption=Linear}
    boneToModify {boneName=dyn_L_backLock_1_04}
    constraintSetup [{transformType=Rotation, targetBone {boneName=dyn_L_frontLock_4_06}}]
    constraintWeights [1]
   }
   "Constraint - base_R_backLock_1 (PPP)"!(ComponentPose, Alpha=1, ConstraintWeights_0=0.6647, ConstraintWeights_1:float=0.2951, ConstraintWeights_2:float=0.0402) -> (Pose) {
    alphaBoolBlend {blendOption=Linear}
    boneToModify {boneName=base_R_backLock_1}
    constraintSetup [{targetBone {boneName=spine_05}}, {targetBone {boneName=clavicle_r}}, {targetBone {boneName=neck_01}}]
    constraintWeights [1,1,1]
   }
   "Constraint - base_R_backLock_2 (PPP)"!(ComponentPose, Alpha=1, ConstraintWeights_0=0.6934, ConstraintWeights_1:float=0.2714, ConstraintWeights_2:float=0.0353) -> (Pose) {
    alphaBoolBlend {blendOption=Linear}
    boneToModify {boneName=base_R_backLock_2}
    constraintSetup [{targetBone {boneName=spine_05}}, {targetBone {boneName=clavicle_r}}, {targetBone {boneName=neck_01}}]
    constraintWeights [1,1,1]
   }
   "Transform (Modify) Bone - Bone: dyn_L_backLock_1_04"(Alpha=1, Translation="-1,0,0", Rotation="0,0,0", Scale="1,1,1") {
    rotationMode=BMM_Additive
    rotationSpace=BCS_BoneSpace
    scaleMode=BMM_Additive
    scaleSpace=BCS_BoneSpace
    translationMode=BMM_Additive
    translationSpace=BCS_BoneSpace
    alphaBoolBlend {blendOption=Linear}
    boneToModify {boneName=dyn_L_backLock_1_04}
    translation {x=-1}
   }
   n1 = "Transform (Modify) Bone - Bone: dyn_R_frontLock_3_06"(Alpha=1, Translation="0,0,0", Rotation="12,0,0", Scale="1,1,1") {
    rotationMode=BMM_Additive
    rotationSpace=BCS_BoneSpace
    scaleMode=BMM_Additive
    scaleSpace=BCS_BoneSpace
    translationMode=BMM_Additive
    translationSpace=BCS_BoneSpace
    alphaBoolBlend {blendOption=Linear}
    boneToModify {boneName=dyn_R_frontLock_3_06}
    rotation {pitch=12}
   }
   "Transform (Modify) Bone - Bone: dyn_L_backLock_1_04"(ComponentPose=@n1, Alpha=1, Translation="0,0,0", Rotation="-0.067,5.000,-0.011", Scale="1,1,1") {
    rotationMode=BMM_Additive
    rotationSpace=BCS_BoneSpace
    scaleMode=BMM_Additive
    scaleSpace=BCS_BoneSpace
    translationMode=BMM_Additive
    translationSpace=BCS_BoneSpace
    alphaBoolBlend {blendOption=Linear}
    boneToModify {boneName=dyn_L_backLock_1_04}
    rotation=<rot -0.067 5 -0.011>
   }
   n2 = RigidBody!(
    ComponentPose=Local To Component(LocalPose=Use cached pose 'Control Rig Pose'())
    Alpha
    OverrideWorldGravity:Vector=Get GravityOverride()
    ExternalForce
   ) -> (Pose) {
    bClampLinearTranslationLimitToRefPose=true
    bOverrideWorldGravity=true
    overridePhysicsAsset=&/M_MED_EssayViewMyth_Pier_FaceAcc_Back_Physics
    simulationSpace=BaseBoneSpace
    alphaBoolBlend {blendOption=Linear}
    baseBoneRef {boneName=spine_05}
    simSpaceSettings {
     dampingAlpha=0
     maxAngularAcceleration=16
     maxAngularVelocity=8
     maxLinearAcceleration=800
     maxLinearVelocity=600
     worldAlpha=1
     externalLinearDragV=<2 0.2 0.2>
    }
   }
   n3 = "Transform (Modify) Bone - Bone: base_L_frontLock_1"!(ComponentPose=@n2, bAlphaBoolEnabled:bool=True, Translation="-0.5,-0.5,0", Rotation="0,-3,0", Scale="1,1,1") -> (Pose) {
    alphaInputType=Bool
    rotationMode=BMM_Additive
    rotationSpace=BCS_BoneSpace
    translationMode=BMM_Additive
    translationSpace=BCS_BoneSpace
    alphaBoolBlend {blendInTime=0.25, blendOption=Linear, blendOutTime=0.25}
    boneToModify {boneName=base_L_frontLock_1}
    rotation {yaw=-3}
    translation=<-0.5 -0.5>
   }
   n4 = "Transform (Modify) Bone - Bone: base_L_frontLock_2"!(ComponentPose=@n3, bAlphaBoolEnabled:bool=True, Translation="-0.5,-0.5,0", Rotation="0,-6,0", Scale="1,1,1") -> (Pose) {
    alphaInputType=Bool
    rotationMode=BMM_Additive
    rotationSpace=BCS_BoneSpace
    translationMode=BMM_Additive
    translationSpace=BCS_BoneSpace
    alphaBoolBlend {blendInTime=0.25, blendOption=Linear, blendOutTime=0.25}
    boneToModify {boneName=base_L_frontLock_2}
    rotation {yaw=-6}
    translation=<-0.5 -0.5>
   }
   n5 = "Transform (Modify) Bone - Bone: base_L_frontLock_3"!(ComponentPose=@n4, bAlphaBoolEnabled:bool=True, Translation="-0.5,-0.5,0", Rotation="0,-5,0", Scale="1,1,1") -> (Pose) {
    alphaInputType=Bool
    rotationMode=BMM_Additive
    rotationSpace=BCS_BoneSpace
    translationMode=BMM_Additive
    translationSpace=BCS_BoneSpace
    alphaBoolBlend {blendInTime=0.1, blendOption=Linear, blendOutTime=0.1}
    boneToModify {boneName=base_L_frontLock_3}
    rotation {yaw=-5}
    translation=<-0.5 -0.5>
   }
   n6 = "Transform (Modify) Bone - Bone: dyn_L_frontLock_3_06"!(ComponentPose=@n5, bAlphaBoolEnabled:bool=True, Translation="0,0,0", Rotation="0,15,0", Scale="1,1,1") -> (Pose) {
    alphaInputType=Bool
    rotationMode=BMM_Additive
    rotationSpace=BCS_BoneSpace
    scaleMode=BMM_Additive
    scaleSpace=BCS_BoneSpace
    translationMode=BMM_Additive
    translationSpace=BCS_BoneSpace
    alphaBoolBlend {blendInTime=1, blendOption=Linear, blendOutTime=1}
    boneToModify {boneName=dyn_L_frontLock_3_06}
    rotation {yaw=15}
   }
   n7 = Component To Local(ComponentPose=@n6)
   n8 = RigidBody(ComponentPose=Local To Component(LocalPose=Use cached pose 'Control Rig Pose'()), OverrideWorldGravity=Get GravityOverride()) {
    bClampLinearTranslationLimitToRefPose=true
    bOverrideWorldGravity=true
    overridePhysicsAsset=&/M_MED_EssayViewMyth_FaceAcc_Back_Physics
    simulationSpace=BaseBoneSpace
    alphaBoolBlend {blendOption=Linear}
    baseBoneRef {boneName=spine_05}
    simSpaceSettings {
     dampingAlpha=0
     maxAngularAcceleration=16
     maxAngularVelocity=8
     maxLinearAcceleration=800
     maxLinearVelocity=600
     worldAlpha=1
     externalLinearDragV=<2 0.2 0.2>
    }
   }
   n9 = "Transform (Modify) Bone - Bone: base_L_backLock_1"(ComponentPose=@n8, Alpha=1, Translation="0,0,0", Rotation="0,0,-30", Scale="1,1,1") {
    rotationMode=BMM_Additive
    rotationSpace=BCS_BoneSpace
    scaleMode=BMM_Additive
    scaleSpace=BCS_BoneSpace
    translationMode=BMM_Additive
    translationSpace=BCS_BoneSpace
    alphaBoolBlend {blendOption=Linear}
    boneToModify {boneName=base_L_backLock_1}
    rotation {roll=-30}
   }
   n10 = "Transform (Modify) Bone - Bone: base_L_frontLock_3"!(ComponentPose=@n9, Alpha:float=1, Translation="0,0,0", Rotation="-0.308,4.996,0.052", Scale="1,1,1") -> (Pose) {
    rotationMode=BMM_Additive
    rotationSpace=BCS_BoneSpace
    scaleMode=BMM_Additive
    scaleSpace=BCS_BoneSpace
    translationMode=BMM_Additive
    translationSpace=BCS_BoneSpace
    alphaBoolBlend {blendOption=Linear}
    boneToModify {boneName=base_L_frontLock_3}
    rotation=<rot -0.3079 4.996 0.052>
   }
   n11 = Component To Local(ComponentPose=@n10)
   n12 = Get isEssayPier() -> (isEssayPier:bool)
   n13 = "Constraint - dyn_L_backLock_2_01 (PPPPP)"!(
    ComponentPose=Local To Component(LocalPose=Use cached pose 'Control Rig Pose'())
    Alpha=1
    ConstraintWeights_0=0
    ConstraintWeights_1:float=0
    ConstraintWeights_2:float=0.34
    ConstraintWeights_3:float=0.33
    ConstraintWeights_4:float=0.33
   ) -> (Pose) {
    alphaBoolBlend {blendOption=Linear}
    boneToModify {boneName=dyn_L_backLock_2_01}
    constraintSetup [
     {targetBone {boneName=spine_05}}
     {targetBone {boneName=clavicle_l}}
     {targetBone {boneName=neck_01}}
     {targetBone {boneName=head}}
     {targetBone {boneName=neck_02}}
    ]
    constraintWeights [1,1,1,1,1]
   }
   n14 = "Constraint - dyn_L_backLock_2_03 (PPPPP)"!(
    ComponentPose=@n13
    Alpha=1
    ConstraintWeights_0=0.33
    ConstraintWeights_1:float=0
    ConstraintWeights_2:float=0.33
    ConstraintWeights_3:float=0.34
    ConstraintWeights_4:float=0
   ) -> (Pose) {
    alphaBoolBlend {blendOption=Linear}
    boneToModify {boneName=dyn_L_backLock_2_03}
    constraintSetup [
     {targetBone {boneName=spine_05}}
     {targetBone {boneName=clavicle_l}}
     {targetBone {boneName=neck_01}}
     {targetBone {boneName=head}}
     {targetBone {boneName=spine_04}}
    ]
    constraintWeights [1,1,1,1,1]
   }
   n15 = "Constraint - dyn_L_backLock_2_05 (PPPPP)"!(
    ComponentPose=@n14
    Alpha=1
    ConstraintWeights_0=0.7
    ConstraintWeights_1:float=0
    ConstraintWeights_2:float=0.3
    ConstraintWeights_3:float=0
    ConstraintWeights_4:float=0
   ) -> (Pose) {
    alphaBoolBlend {blendOption=Linear}
    boneToModify {boneName=dyn_L_backLock_2_05}
    constraintSetup [
     {targetBone {boneName=spine_05}}
     {targetBone {boneName=clavicle_l}}
     {targetBone {boneName=neck_01}}
     {targetBone {boneName=spine_03}}
     {targetBone {boneName=spine_04}}
    ]
    constraintWeights [1,1,1,1,1]
   }
   n16 = "Constraint - dyn_L_backLock_2_06 (PPPPP)"!(
    ComponentPose=@n15
    Alpha=1
    ConstraintWeights_0=0.5
    ConstraintWeights_1:float=0.5
    ConstraintWeights_2:float=0
    ConstraintWeights_3:float=0
    ConstraintWeights_4:float=0
   ) -> (Pose) {
    alphaBoolBlend {blendOption=Linear}
    boneToModify {boneName=dyn_L_backLock_2_06}
    constraintSetup [
     {targetBone {boneName=spine_05}}
     {targetBone {boneName=clavicle_l}}
     {targetBone {boneName=spine_03}}
     {targetBone {boneName=spine_04}}
     {targetBone {boneName=neck_01}}
    ]
    constraintWeights [1,1,1,1,1]
   }
   n17 = "Constraint - dyn_M_backLock_1_01 (PPPP)"!(
    ComponentPose=@n16
    Alpha=1
    ConstraintWeights_0=0
    ConstraintWeights_1:float=0.1
    ConstraintWeights_2:float=0
    ConstraintWeights_3:float=0.9
   ) -> (Pose) {
    alphaBoolBlend {blendOption=Linear}
    boneToModify {boneName=dyn_M_backLock_1_01}
    constraintSetup [{targetBone {boneName=spine_05}}, {targetBone {boneName=neck_02}}, {targetBone {boneName=neck_01}}, {targetBone {boneName=head}}]
    constraintWeights [1,1,1,1]
   }
   n18 = "Constraint - dyn_M_backLock_1_03 (PPPP)"!(
    ComponentPose=@n17
    Alpha=1
    ConstraintWeights_0=0.25
    ConstraintWeights_1:float=0.25
    ConstraintWeights_2:float=0.25
    ConstraintWeights_3:float=0.25
   ) -> (Pose) {
    alphaBoolBlend {blendOption=Linear}
    boneToModify {boneName=dyn_M_backLock_1_03}
    constraintSetup [{targetBone {boneName=spine_05}}, {targetBone {boneName=head}}, {targetBone {boneName=neck_01}}, {targetBone {boneName=neck_02}}]
    constraintWeights [1,1,1,1]
   }
   n19 = "Constraint - dyn_M_backLock_1_05 (PPPPPP)"!(
    ComponentPose=@n18
    Alpha=1
    ConstraintWeights_0=0.8
    ConstraintWeights_1:float=0
    ConstraintWeights_2:float=0
    ConstraintWeights_3:float=0
    ConstraintWeights_4:float=0
    ConstraintWeights_5:float=0.2
   ) -> (Pose) {
    alphaBoolBlend {blendOption=Linear}
    boneToModify {boneName=dyn_M_backLock_1_05}
    constraintSetup [
     {targetBone {boneName=spine_05}}
     {targetBone {boneName=clavicle_r}}
     {targetBone {boneName=spine_03}}
     {targetBone {boneName=spine_04}}
     {targetBone {boneName=clavicle_l}}
     {targetBone {boneName=neck_01}}
    ]
    constraintWeights [1,1,1,1,1,1]
   }
   n20 = "Constraint - dyn_M_backLock_2_01 (PPP)"!(ComponentPose=@n19, Alpha=1, ConstraintWeights_0=0.6, ConstraintWeights_1:float=0.2, ConstraintWeights_2:float=0.2) -> (Pose) {
    alphaBoolBlend {blendOption=Linear}
    boneToModify {boneName=dyn_M_backLock_2_01}
    constraintSetup [{targetBone {boneName=spine_05}}, {targetBone {boneName=neck_02}}, {targetBone {boneName=head}}]
    constraintWeights [1,1,1]
   }
   n21 = "Constraint - dyn_M_backLock_2_03 (PPPP)"!(
    ComponentPose=@n20
    Alpha=1
    ConstraintWeights_0=0.8
    ConstraintWeights_1:float=0
    ConstraintWeights_2:float=0.1
    ConstraintWeights_3:float=0.1
   ) -> (Pose) {
    alphaBoolBlend {blendOption=Linear}
    boneToModify {boneName=dyn_M_backLock_2_03}
    constraintSetup [{targetBone {boneName=spine_05}}, {targetBone {boneName=head}}, {targetBone {boneName=neck_01}}, {targetBone {boneName=neck_02}}]
    constraintWeights [1,1,1,1]
   }
   n22 = "Constraint - dyn_M_backLock_2_05 (PPPPPP)"!(
    ComponentPose=@n21
    Alpha=1
    ConstraintWeights_0=0.8
    ConstraintWeights_1:float=0
    ConstraintWeights_2:float=0
    ConstraintWeights_3:float=0
    ConstraintWeights_4:float=0
    ConstraintWeights_5:float=0.2
   ) -> (Pose) {
    alphaBoolBlend {blendOption=Linear}
    boneToModify {boneName=dyn_M_backLock_2_05}
    constraintSetup [
     {targetBone {boneName=spine_05}}
     {targetBone {boneName=clavicle_r}}
     {targetBone {boneName=spine_03}}
     {targetBone {boneName=spine_04}}
     {targetBone {boneName=clavicle_l}}
     {targetBone {boneName=neck_01}}
    ]
    constraintWeights [1,1,1,1,1,1]
   }
   n23 = "Constraint - dyn_M_backLock_2_06 (PPPPP)"!(
    ComponentPose=@n22
    Alpha=1
    ConstraintWeights_0=0.9
    ConstraintWeights_1:float=0
    ConstraintWeights_2:float=0
    ConstraintWeights_3:float=0
    ConstraintWeights_4:float=0.1
   ) -> (Pose) {
    alphaBoolBlend {blendOption=Linear}
    boneToModify {boneName=dyn_M_backLock_2_06}
    constraintSetup [
     {targetBone {boneName=spine_05}}
     {targetBone {boneName=clavicle_r}}
     {targetBone {boneName=spine_03}}
     {targetBone {boneName=spine_04}}
     {targetBone {boneName=neck_01}}
    ]
    constraintWeights [1,1,1,1,1]
   }
   n24 = "Constraint - dyn_M_backLock_3_01 (PPPPP)"!(
    ComponentPose=@n23
    Alpha=1
    ConstraintWeights_0=0.3
    ConstraintWeights_1:float=0.1
    ConstraintWeights_2:float=0.1
    ConstraintWeights_3:float=0.5
    ConstraintWeights_4:float=0
   ) -> (Pose) {
    alphaBoolBlend {blendOption=Linear}
    boneToModify {boneName=dyn_M_backLock_3_01}
    constraintSetup [
     {targetBone {boneName=spine_05}}
     {targetBone {boneName=neck_02}}
     {targetBone {boneName=neck_01}}
     {targetBone {boneName=head}}
     {targetBone {boneName=spine_04}}
    ]
    constraintWeights [1,1,1,1,1]
   }
   n25 = "Constraint - dyn_M_backLock_3_03 (PPPP)"!(
    ComponentPose=@n24
    Alpha=1
    ConstraintWeights_0=0.5
    ConstraintWeights_1:float=0.25
    ConstraintWeights_2:float=0.25
    ConstraintWeights_3:float=0
   ) -> (Pose) {
    alphaBoolBlend {blendOption=Linear}
    boneToModify {boneName=dyn_M_backLock_3_03}
    constraintSetup [{targetBone {boneName=spine_05}}, {targetBone {boneName=neck_01}}, {targetBone {boneName=neck_02}}, {targetBone {boneName=head}}]
    constraintWeights [1,0,1,1]
   }
   n26 = "Constraint - dyn_M_backLock_3_05 (PPPPPP)"!(
    ComponentPose=@n25
    Alpha=1
    ConstraintWeights_0=0.9
    ConstraintWeights_1:float=0
    ConstraintWeights_2:float=0
    ConstraintWeights_3:float=0
    ConstraintWeights_4:float=0.1
    ConstraintWeights_5:float=0
   ) -> (Pose) {
    alphaBoolBlend {blendOption=Linear}
    boneToModify {boneName=dyn_M_backLock_3_05}
    constraintSetup [
     {targetBone {boneName=spine_05}}
     {targetBone {boneName=clavicle_l}}
     {targetBone {boneName=spine_03}}
     {targetBone {boneName=clavicle_r}}
     {targetBone {boneName=neck_01}}
     {targetBone {boneName=spine_04}}
    ]
    constraintWeights [1,1,1,1,1,1]
   }
   n27 = "Constraint - dyn_M_backLock_3_06 (PPPPP)"!(
    ComponentPose=@n26
    Alpha=1
    ConstraintWeights_0=0.9
    ConstraintWeights_1:float=0
    ConstraintWeights_2:float=0
    ConstraintWeights_3:float=0
    ConstraintWeights_4:float=0.1
   ) -> (Pose) {
    alphaBoolBlend {blendOption=Linear}
    boneToModify {boneName=dyn_M_backLock_3_06}
    constraintSetup [
     {targetBone {boneName=spine_05}}
     {targetBone {boneName=clavicle_l}}
     {targetBone {boneName=spine_03}}
     {targetBone {boneName=spine_04}}
     {targetBone {boneName=neck_01}}
    ]
    constraintWeights [1,1,1,1,1]
   }
   n28 = "Constraint - dyn_R_backLock_1_01 (PPP)"!(ComponentPose=@n27, Alpha=1, ConstraintWeights_0=0.8, ConstraintWeights_1:float=0.1, ConstraintWeights_2:float=0.1) -> (Pose) {
    alphaBoolBlend {blendOption=Linear}
    boneToModify {boneName=dyn_R_backLock_1_01}
    constraintSetup [{targetBone {boneName=head}}, {targetBone {boneName=neck_02}}, {targetBone {boneName=neck_01}}]
    constraintWeights [1,1,1]
   }
   n29 = "Constraint - dyn_R_backLock_1_03 (PPPP)"!(
    ComponentPose=@n28
    Alpha=1
    ConstraintWeights_0=0
    ConstraintWeights_1:float=0.33
    ConstraintWeights_2:float=0.33
    ConstraintWeights_3:float=0.34
   ) -> (Pose) {
    alphaBoolBlend {blendOption=Linear}
    boneToModify {boneName=dyn_R_backLock_1_03}
    constraintSetup [{targetBone {boneName=head}}, {targetBone {boneName=clavicle_r}}, {targetBone {boneName=neck_01}}, {targetBone {boneName=neck_02}}]
    constraintWeights [1,1,1,1]
   }
   n30 = "Constraint - dyn_R_backLock_1_05 (PPPPP)"!(
    ComponentPose=@n29
    Alpha=1
    ConstraintWeights_0=0
    ConstraintWeights_1:float=0.9
    ConstraintWeights_2:float=0
    ConstraintWeights_3:float=0
    ConstraintWeights_4:float=0.1
   ) -> (Pose) {
    alphaBoolBlend {blendOption=Linear}
    boneToModify {boneName=dyn_R_backLock_1_05}
    constraintSetup [
     {targetBone {boneName=spine_05}}
     {targetBone {boneName=clavicle_r}}
     {targetBone {boneName=spine_03}}
     {targetBone {boneName=spine_04}}
     {targetBone {boneName=neck_01}}
    ]
    constraintWeights [1,1,1,1,1]
   }
   n31 = "Constraint - dyn_R_backLock_1_06 (PPPP)"!(
    ComponentPose=@n30
    Alpha=1
    ConstraintWeights_0=0.1
    ConstraintWeights_1:float=0.9
    ConstraintWeights_2:float=0
    ConstraintWeights_3:float=0
   ) -> (Pose) {
    alphaBoolBlend {blendOption=Linear}
    boneToModify {boneName=dyn_R_backLock_1_06}
    constraintSetup [{targetBone {boneName=spine_05}}, {targetBone {boneName=clavicle_r}}, {targetBone {boneName=spine_03}}, {targetBone {boneName=spine_04}}]
    constraintWeights [1,1,1,1]
   }
   n32 = "Constraint - dyn_R_backLock_2_01 (PPPPP)"!(
    ComponentPose=@n31
    Alpha=1
    ConstraintWeights_0=0
    ConstraintWeights_1:float=0.9
    ConstraintWeights_2:float=0
    ConstraintWeights_3:float=0.1
    ConstraintWeights_4:float=0
   ) -> (Pose) {
    alphaBoolBlend {blendOption=Linear}
    boneToModify {boneName=dyn_R_backLock_2_01}
    constraintSetup [
     {targetBone {boneName=spine_05}}
     {targetBone {boneName=head}}
     {targetBone {boneName=neck_01}}
     {targetBone {boneName=neck_02}}
     {targetBone {boneName=clavicle_r}}
    ]
    constraintWeights [1,1,1,1,1]
   }
   n33 = "Constraint - dyn_R_backLock_2_03 (PPPPP)"!(
    ComponentPose=@n32
    Alpha=1
    ConstraintWeights_0=0
    ConstraintWeights_1:float=0.25
    ConstraintWeights_2:float=0.25
    ConstraintWeights_3:float=0.25
    ConstraintWeights_4:float=0.25
   ) -> (Pose) {
    alphaBoolBlend {blendOption=Linear}
    boneToModify {boneName=dyn_R_backLock_2_03}
    constraintSetup [
     {targetBone {boneName=spine_05}}
     {targetBone {boneName=clavicle_r}}
     {targetBone {boneName=neck_01}}
     {targetBone {boneName=neck_02}}
     {targetBone {boneName=head}}
    ]
    constraintWeights [1,1,1,1,1]
   }
   n34 = "Constraint - dyn_R_backLock_2_05 (PP)"!(ComponentPose=@n33, Alpha=1, ConstraintWeights_0=0.1, ConstraintWeights_1:float=0.9) -> (Pose) {
    alphaBoolBlend {blendOption=Linear}
    boneToModify {boneName=dyn_R_backLock_2_05}
    constraintSetup [{targetBone {boneName=spine_05}}, {targetBone {boneName=clavicle_r}}]
    constraintWeights [1,1]
   }
   n35 = "Constraint - dyn_R_backLock_2_06 (PPPPP)"!(
    ComponentPose=@n34
    Alpha=1
    ConstraintWeights_0=0.1
    ConstraintWeights_1:float=0.9
    ConstraintWeights_2:float=0
    ConstraintWeights_3:float=0
    ConstraintWeights_4:float=0
   ) -> (Pose) {
    alphaBoolBlend {blendOption=Linear}
    boneToModify {boneName=dyn_R_backLock_2_06}
    constraintSetup [
     {targetBone {boneName=spine_05}}
     {targetBone {boneName=clavicle_r}}
     {targetBone {boneName=spine_03}}
     {targetBone {boneName=spine_04}}
     {targetBone {boneName=neck_01}}
    ]
    constraintWeights [1,1,1,1,1]
   }
   n36 = RigidBody(ComponentPose=@n35, OverrideWorldGravity=Get GravityOverride()) {
    bClampLinearTranslationLimitToRefPose=true
    bOverrideWorldGravity=true
    overridePhysicsAsset=&/M_MED_EssayViewMyth_FaceAcc_BackBling_Physics
    simulationSpace=BaseBoneSpace
    alphaBoolBlend {blendOption=Linear}
    baseBoneRef {boneName=spine_05}
    simSpaceSettings {
     dampingAlpha=0
     maxAngularAcceleration=16
     maxAngularVelocity=8
     maxLinearAcceleration=800
     maxLinearVelocity=600
     worldAlpha=1
     externalLinearDragV=<2 0.2 0.2>
    }
   }
   Output Pose(Result=Blend Poses by bool(
     BlendPose_0=Component To Local(ComponentPose=@n36)
     BlendPose_1=Blend Poses by bool(BlendPose_0=@n7, BlendPose_1=@n11, bActiveValue=@n12) {blendType=Linear}
     bActiveValue=Get Backbling_or_FrontEndSkydive() -> (Backbling_or_FrontEndSkydive:bool)
    ) {blendType=Linear}) {layerGroup=None}
   n37 = Get Body_Mesh()
   Save cached pose 'Start Pose'(Pose=Layered blend per bone(
     BasePose=Copy Pose From Mesh() {bUseAttachedParent=true}
     BlendPoses_0=Copy Pose From Mesh(SourceMeshComponent=@n37) {bUseAttachedParent=true}
    ) {layerSetup [{branchFilters [{boneName=upperarm_l}, {boneName=upperarm_r}]}]}) {cacheName=Start Pose}
   n38 = Control Rig!(
    Source=Use cached pose 'Start Pose'()
    Alpha
    Is Crouching:bool=Get bIsCrouching()
    Is Crouch Walking:bool=Get IsCrouchWalking() -> (IsCrouchWalking:bool)
    Crouch Aim Walking:bool=Get IsCrouchAimingMoving() -> (IsCrouchAimingMoving:bool)
    Is Crouch Aiming:bool=Get IsCrouchAiming() -> (IsCrouchAiming:bool)
   ) -> (Pose) {
    alphaBoolBlend {blendOption=Linear}
    controlRigAssetReference {blueprintRigClass=&/M_MED_EssayViewMyth_FaceAcc_CtrlRig.M_MED_EssayViewMyth_FaceAcc_CtrlRig_C}
   }
   Save cached pose 'Control Rig Pose'(Pose=@n38) {cacheName=Control Rig Pose}
  ]}
 EventGraph {nodes [
   Event BlueprintUpdateAnimation()
   > n1 = Sequence!(execute) -> (then_0, then_1, then_2:Exec)
   > Set IsCrouchAiming(IsCrouchAiming:bool=AND Boolean(A=Get bIsCrouching(), B=Get bIsTargeting())) -> (Output_Get:bool)
   > Set IsCrouchWalking(IsCrouchWalking:bool=AND Boolean(A=Get bIsCrouching(), B=Get bIsMoving2D())) -> (Output_Get:bool)
   > Set IsCrouchAimingMoving(IsCrouchAimingMoving:bool=AND Boolean!(A=Get bIsCrouching(), B=Get bIsTargeting(), C:bool=Get bIsMoving2D()) -> (ReturnValue)) -> (Output_Get:bool)
   > n2 = Branch(Condition=Get bIsSkyDiving())
   > Set GravityOverride(GravityOverride="0,0,500")
   > n3 = Branch(Condition=Get is Wearing Backbling() -> (is Wearing Backbling:bool))
   > Set GravityOverride(GravityOverride="0,0,-700")
   Set GravityOverride(execute=@n3.else, GravityOverride="0,0,500")
   > n4 = Branch(Condition=Get bIsParachuteOpened())
   > Set GravityOverride(GravityOverride="0,0,-980")
   Set GravityOverride(execute=@n4.else, GravityOverride="0,0,50")
   Set GravityOverride(execute=@n2.else, GravityOverride="0,0,-980")
   n5 = Cast To FortPlayerPawn(execute=@n1.then_2, Object=GetOwningActor())
   > n6 = Set Character_0(Character_0:Athena Character Item Definition Object Reference=GetItemDefinitionTypeFromLoadoutWithActor!(
     Actor:Actor Object Reference=@n5
     LoadoutSlot:Gameplay Tag Structure=(TagName="LoadoutSlot.Character")
     CosmeticType:Athena Cosmetic Item Definition Class Reference=/Script/FortniteGame.AthenaCharacterItemDefinition
    ) -> (OutCosmeticItemDef:Athena Character Item Definition Object Reference)) -> (Output_Get:Athena Character Item Definition Object Reference)
   > Set isEssayPier(isEssayPier:bool="Equal (Object)"(A=@n6, B=/BRCosmetics/Athena/Items/Cosmetics/Characters/Character_EssayViewPier)) -> (Output_Get:bool)
   > Set isEssayViewMyth(isEssayViewMyth:bool="Equal (Object)"(A=@n6, B=/BRCosmetics/Athena/Items/Cosmetics/Characters/Character_EssayViewMyth)) -> (Output_Get:bool)
   n7 = CheckForBackbling()
   Set is Wearing Backbling(
    execute=@n1.then_1
    is Wearing Backbling:bool=AND Boolean(A=@n7, B=NOT Boolean(A=@n7.bIsBackblingACape), C=NOT Boolean(A=@n7.bIsBackblingLong))
   ) -> (Output_Get:bool)
   > n8 = Set is Wearing Cape(is Wearing Cape:bool=OR Boolean(A=@n7.bIsBackblingACape, B=@n7.bIsBackblingLong)) -> (Output_Get:bool)
   n9 = Get bIsInFrontEnd()
   n10 = Get bIsSkyDiving()
   n11 = Get is Wearing Backbling()
   > Set Backbling_or_FrontEndSkydive(Backbling_or_FrontEndSkydive:bool=OR Boolean(A=OR Boolean(A=@n8, B=@n11), B=AND Boolean!(A=@n9, B=@n10) -> (ReturnValue))) -> (Output_Get:bool)
   n12 = Event OnCharacterCustomizationLoaded()
   > Set Body_Mesh(Body_Mesh=@n12.CustomBaseMesh)
   > Set Skel Mesh(Skel Mesh=Get Array Item with EFort Custom Part Type(Array=@n12.CustomMeshParts, PartType=Face))
   Get bIsSkyDiving()
  ]}
]
variables [
 Backbling_or_FrontEndSkydive {type=boolean}
 Body_Mesh {type=SkeletalMeshComponent}
 Character {type=AthenaCharacterItemDefinition, default=null}
 Character_0 {type=AthenaCharacterItemDefinition, default=null}
 GravityOverride {type=Vector}
 Input Object {type=Object, default=&Character_EssayViewPier}
 is Wearing Backbling {type=boolean}
 is Wearing Cape {type=boolean}
 IsCrouchAiming {type=boolean}
 IsCrouchAimingMoving {type=boolean}
 IsCrouchWalking {type=boolean}
 isEssayPier {type=boolean}
 isEssayViewMyth {type=boolean}
 Skel Mesh {type=SkeletalMeshComponent}
]
